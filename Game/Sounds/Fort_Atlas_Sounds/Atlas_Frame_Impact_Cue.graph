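# Sound cue for Atlas frame impacts, layering car-impact hits with electric wall-trap zaps.
# A quality-level root picks between a full mixer of randomized variations, a slimmer two-by-two mixer, or a single optimized impact wave for the two lowest tiers.
# Modulators vary each branch, and the cue raises overall pitch.
# Medium-object demolition attenuation applies, with a concurrency override that stops the oldest instance when the small voice cap is hit.
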SoundCue /Game/Sounds/Fort_Atlas_Sounds/Atlas_Frame_Impact_Cue
root=QualityLevel_0
nodes [
 {id=Mixer_0, children [Modulator_1,Modulator_2], properties {inputVolume [1,1]}}
 {id=Mixer_1, children [Modulator_0,Modulator_3], properties {inputVolume [1,1]}}
 {id=Modulator_0, children [Random_4], properties {pitchMax=1.1, pitchMin=0.9}}
 {id=Modulator_1, children [Random_1]}
 {id=Modulator_2, children [Random_0]}
 {id=Modulator_3, children [Random_5], properties {pitchMax=1.12, pitchMin=0.92}}
 {id=Modulator_8, children [WavePlayer_21], properties {pitchMax=1.1, pitchMin=0.9}}
 {id=Modulator_9, children [WavePlayer_22], properties {pitchMax=1.1, pitchMin=0.9}}
 {id=QualityLevel_0, children [Mixer_0,Mixer_1,Modulator_8,Modulator_9]}
 {id=Random_0, children [WavePlayer_8, WavePlayer_9, WavePlayer_10, WavePlayer_11, WavePlayer_12], properties {weights [1,1,1,1,1]}}
 {id=Random_1, children [WavePlayer_17, WavePlayer_18, WavePlayer_19, WavePlayer_20], properties {weights [1,1,1,1]}}
 {id=Random_4, children [WavePlayer_2,WavePlayer_3], properties {bRandomizeWithoutReplacement=false, weights [1,1]}}
 {id=Random_5, children [WavePlayer_0,WavePlayer_1], properties {weights [1,1]}}
 {id=WavePlayer_0, properties {soundWaveAssetPtr=/Game/Sounds/Fort_Impact_Sounds/Demolition/Car/Car_Impact_Hard5}}
 {id=WavePlayer_1, properties {soundWaveAssetPtr=/Game/Sounds/Fort_Impact_Sounds/Demolition/Car/Car_Impact_Med4}}
 {id=WavePlayer_10, properties {soundWaveAssetPtr=/Game/Sounds/Fort_Impact_Sounds/Demolition/Car/Car_Impact_Med3}}
 {id=WavePlayer_11, properties {soundWaveAssetPtr=/Game/Sounds/Fort_Impact_Sounds/Demolition/Car/Car_Impact_Med2}}
 {id=WavePlayer_12, properties {soundWaveAssetPtr=/Game/Sounds/Fort_Impact_Sounds/Demolition/Car/Car_Impact_Med1}}
 {id=WavePlayer_17, properties {soundWaveAssetPtr=/Game/Sounds/Fort_Traps/Wall_Electric/WallTrap_Electric_Fire_mono_002}}
 {id=WavePlayer_18, properties {soundWaveAssetPtr=/Game/Sounds/Fort_Traps/Wall_Electric/WallTrap_Electric_Fire_mono_001}}
 {id=WavePlayer_19, properties {soundWaveAssetPtr=/Game/Sounds/Fort_Traps/Wall_Electric/WallTrap_Electric_Fire_mono_004}}
 {id=WavePlayer_2, properties {soundWaveAssetPtr=&WallTrap_Electric_Fire_mono_002}}
 {id=WavePlayer_20, properties {soundWaveAssetPtr=/Game/Sounds/Fort_Traps/Wall_Electric/WallTrap_Electric_Fire_mono_003}}
 {id=WavePlayer_21, properties {soundWaveAssetPtr=&/Atlast_Frame_Impact_opt}}
 {id=WavePlayer_22, properties {soundWaveAssetPtr=&Atlast_Frame_Impact_opt}}
 {id=WavePlayer_3, properties {soundWaveAssetPtr=&WallTrap_Electric_Fire_mono_001}}
 {id=WavePlayer_8, properties {soundWaveAssetPtr=&Car_Impact_Hard5}}
 {id=WavePlayer_9, properties {soundWaveAssetPtr=&Car_Impact_Med4}}
]
properties {
 attenuationSettings=/Game/Sounds/Attenuation/Fort_Demolition_Hit_Medium_Sized_Object_Att
 bOverrideConcurrency=true
 maxDistance=4000
 pitchMultiplier=1.3
 soundClassObject=/Game/Sounds/FortSoundClassesAndModes/Fort_Atlas_Sounds_PSM
 volumeMultiplier=0.7
 attenuationOverrides {falloffDistance=3200, attenuationShapeExtents {x=800}}
 concurrencyOverrides {maxCount=4, resolutionRule=StopOldest}
}
